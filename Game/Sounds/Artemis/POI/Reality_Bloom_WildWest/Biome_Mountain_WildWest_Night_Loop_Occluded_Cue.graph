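# Looping night mountain ambience cue, named as the occluded variant for the Wild West Reality Bloom POI, played on the outdoor ambient sound class at reduced volume and pitch.
# A quality switch feeds the quad alpine night loop to its first two levels and the stereo loop to the last two.
# Each wave runs through nested endless envelopers: 30-second pitch modulation, then volume swells looping at 20 seconds, despite keys extending to 60.
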SoundCue /Game/Sounds/Artemis/POI/Reality_Bloom_WildWest/Biome_Mountain_WildWest_Night_Loop_Occluded_Cue
root=QualityLevel_0
nodes [
 {
  id=Enveloper_0
  children [WavePlayer_0]
  properties {
   bLoop=true
   bLoopIndefinitely=true
   loopEnd=30
   pitchCurve {editorCurveData {keys [{interpMode=RCIM_Cubic, time=0}, <at 15.9468 0.9181 RCIM_Cubic>, {interpMode=RCIM_Cubic, time=30}]}}
   volumeCurve {editorCurveData {keys [<at 0 1>]}}
  }
 }
 {
  id=Enveloper_1
  children [WavePlayer_2]
  properties {
   bLoop=true
   bLoopIndefinitely=true
   loopEnd=30
   pitchCurve {editorCurveData {keys [{interpMode=RCIM_Cubic, time=0}, <at 15.9468 0.9181 RCIM_Cubic>, {interpMode=RCIM_Cubic, time=30}]}}
   volumeCurve {editorCurveData {keys [<at 0 1>]}}
  }
 }
 {
  id=Enveloper_2
  children [Enveloper_1]
  properties {
   bLoop=true
   bLoopIndefinitely=true
   loopEnd=20
   pitchCurve {editorCurveData {keys [{time=0}]}}
   volumeCurve {editorCurveData {keys [
      {arriveTangent=-0.01, interpMode=RCIM_Cubic, time=0, value=1}
      {arriveTangent=0.0004, interpMode=RCIM_Cubic, leaveTangent=0.0004, tangentMode=RCTM_User, time=10, value=0.7}
      {arriveTangent=0.0319, interpMode=RCIM_Cubic, leaveTangent=0.0319, tangentMode=RCTM_User, time=20, value=0.8}
      {arriveTangent=-0.0401, interpMode=RCIM_Cubic, leaveTangent=-0.0401, tangentMode=RCTM_User, time=30, value=1}
      {arriveTangent=0, interpMode=RCIM_Cubic, leaveTangent=0, tangentMode=RCTM_User, time=45, value=0.7}
      <at 60 1 RCIM_Cubic>
     ]}}
  }
 }
 {
  id=Enveloper_3
  children [Enveloper_0]
  properties {
   bLoop=true
   bLoopIndefinitely=true
   loopEnd=20
   pitchCurve {editorCurveData {keys [{time=0}]}}
   volumeCurve {editorCurveData {keys [
      {arriveTangent=-0.01, interpMode=RCIM_Cubic, time=0, value=1}
      <at 10 0.9 RCIM_Cubic>
      {arriveTangent=-0.0047, interpMode=RCIM_Cubic, leaveTangent=-0.0047, time=20, value=1}
      {arriveTangent=-0.0228, interpMode=RCIM_Cubic, leaveTangent=-0.0228, tangentMode=RCTM_User, time=30, value=0.8051}
      {arriveTangent=0.0011, interpMode=RCIM_Cubic, leaveTangent=0.0011, tangentMode=RCTM_User, time=45, value=0.7}
      <at 60 1 RCIM_Cubic>
     ]}}
  }
 }
 {id=QualityLevel_0, children [Enveloper_2,Enveloper_2,Enveloper_3,Enveloper_3]}
 {id=WavePlayer_0, properties {bLooping=true, soundWaveAssetPtr=/Game/Sounds/Artemis/Biomes/Mountains/Ambience_Mountains_Alpine_Night_Stereo}}
 {id=WavePlayer_2, properties {bLooping=true, soundWaveAssetPtr=/Game/Sounds/Artemis/Biomes/Mountains/Ambience_Mountains_Alpine_Night_Quad}}
]
properties {
 attenuationSettings=&Fort_Ambient_Athena_Att
 maxDistance=2097152
 pitchMultiplier=0.9
 soundClassObject=&Fort_Ambient_Outdoors_Athena
 volumeMultiplier=0.501
 preEffectBusSends [{}]
}
